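# UI material for a Discovery browser category tile: a flat base color with a border, blended toward a hover color.
# Border thickness is a pixel-count parameter, converted through GetUserInterfaceUV's pixel size to stay constant at any widget size.
# Centered, mirrored UVs feed equal-edge SmoothSteps per axis, summed and saturated into a hard-edged border mask.
# Hover_Animate lerps the whole result, border included, to HoverColor.
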Material /DiscoveryBrowser/Art/M_UI_Discovery_CategoryTile
graph {
 properties {
  bUsedWithStaticMesh=false
  blendMode=BLEND_Translucent
  materialDomain=MD_UI
  pixelDepthOffsetMode=PDOM_Legacy
  usedShadingModels=Default Lit
 }
 nodes [
  n1 = MaterialFunctionCall() {materialFunction=&GetUserInterfaceUV}
  n2 = Subtract(A=@n1.Pixel Size, B=Multiply(A=ScalarParameter() {defaultValue=3, parameterName=PixelSize}) {constB=2})
  n3 = Divide(A=ComponentMask(None=@n2) {r=true}, B=ComponentMask(None=@n1.Pixel Size) {r=true})
  n4 = Abs(None=ConstantBiasScale(None=TextureCoordinate()) {bias=-0.5, scale=2})
  n5 = Divide(A=ComponentMask(None=@n2) {g=true}, B=ComponentMask(None=@n1.Pixel Size) {g=true})
  n6 = Add(
   A=SmoothStep(Min=@n3, Max=@n3, Value=ComponentMask(None=@n4) {r=true})
   B=SmoothStep(Min=@n5, Max=@n5, Value=ComponentMask(None=@n4) {g=true})
  )
  LinearInterpolate(
   A=LinearInterpolate(
    A=VectorParameter() {parameterName=BaseColor, defaultValue=<rgba 0 0.0369 0.3372 1>}.RGB
    B=VectorParameter() {parameterName=BorderColor, defaultValue=<rgba 0 0.0908 1 1>}.RGB
    Alpha=Saturate(None=@n6)
   )
   B=VectorParameter() {parameterName=HoverColor, defaultValue=<rgba 1 1 1 1>}.RGB
   Alpha=ScalarParameter() {group=Animators, parameterName=Hover_Animate, sliderMax=1}
  )
 ]
}
