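# Translucent UI material that paints a gradient background for merged tiles in the Discovery Browser, judging by its name.
# An eased linear gradient blends orange and dark purple, then a cubic-eased vertical UV ramp mixes in crimson.
# An offset, fit-scaled radial gradient, with size and falloff parameters, is inverted to fade that color into near-black outside its glow.
# A CoreUI rounded-mask function provides the rounded-corner shape.
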Material /DiscoveryBrowser/Art/M_UI_TileMergeBackground
graph {
 properties {
  bAutomaticallySetUsageInEditor=false
  bUsedWithStaticMesh=false
  blendMode=BLEND_Translucent
  materialDomain=MD_UI
  usedShadingModels=Default Lit
 }
 nodes [
  n1 = Constant3Vector() {constant=<rgba 0.6927 0.1394 0.0108 1>}
  n2 = Constant3Vector() {constant=<rgba 0.0382 0 0.1 1>}
  n3 = Constant() {r=0.256}
  n4 = Constant() {r=1.2}
  n5 = TextureCoordinate()
  n6 = OneMinus(None=MaterialFunctionCall("Value (0-1)"=MaterialFunctionCall(UVs=@n5, GradientStart=@n3, GradientEnd=@n4) {materialFunction=&MF_UI_LinearGradient}.Result) {materialFunction=&MF_UI_EaseCurves}.EaseInOutSine)
  n7 = ComponentMask(None=@n5) {g=true}
  n8 = MaterialFunctionCall() {materialFunction=&MF_UI_UVScaling}
  n9 = Constant2Vector() {g=1.7469, r=-1.2727}
  n10 = MaterialFunctionCall(
   UV=MaterialFunctionCall(
    UVs=Add(A=@n8.Scaled to Fit, B=@n9) {constB=21}
    Texture Scale=ScalarParameter() {defaultValue=5.5, group=Blur Settings, parameterName=GradientSize}
   ) {materialFunction=&ScaleUVsByCenter}.UVs
   GradientEnd=ScalarParameter() {defaultValue=1.5, group=Blur Settings, parameterName=GradientValue}
  ) {materialFunction=&MF_UI_RadialGradient}
  LinearInterpolate(
   A=LinearInterpolate(
    A=LinearInterpolate(A=@n1, B=@n2, Alpha=@n6)
    B=Constant3Vector() {constant=<rgba 0.4167 0.0072 0.0303 1>}
    Alpha=MaterialFunctionCall("Value (0-1)"=@n7) {materialFunction=&MF_UI_EaseCurves}.EaseInCubic
   )
   B=Constant3Vector() {constant=<rgba 0.0022 0 0.006 1>}
   Alpha=Saturate(None=OneMinus(None=@n10.Smooth))
  )
  MaterialFunctionCall(Corner Radius=Constant() {r=10}, Radius in Pixels=Constant() {r=1}) {materialFunction=&MF_UI_CoreUI_RoundedMask}
 ]
}
