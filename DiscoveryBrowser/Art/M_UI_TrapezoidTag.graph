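# Translucent UI material that draws a two-color trapezoid tag, by its name for the Discovery Browser.
# Two MF_BoxSkew calls, one flipped, are blended across a hard center split along U to form the shape, skewed by angle, scale and minimum-width parameters.
# A SmoothStep with adjustable smoothness creates the edge mask, with aspect ratios derived from UV channel 3.
# The ColorChange scalar blends ColorA to ColorB, and the mask multiplies that color.
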Material /DiscoveryBrowser/Art/M_UI_TrapezoidTag
graph {
 properties {
  bUsedWithStaticMesh=false
  blendMode=BLEND_Translucent
  materialDomain=MD_UI
  pixelDepthOffsetMode=PDOM_Legacy
  usedShadingModels=Default Lit
 }
 nodes [
  n1 = ScalarParameter() {parameterName=ColorChange}
  n2 = SmoothStep(
   Min=Constant()
   Max=Multiply(A=ScalarParameter() {defaultValue=0.01, parameterName=Smoothness}, B=NamedRerouteUsage())
   Value=NamedRerouteUsage()
  )
  Multiply(
   A=LinearInterpolate(
    A=VectorParameter() {parameterName=ColorA, defaultValue=<rgba 0 0.0908 1 1>}.RGB
    B=VectorParameter() {parameterName=ColorB, defaultValue=<rgba 0.0009 0.4452 0.9823 1>}.RGB
    Alpha=Saturate(None=@n1)
   )
   B=@n2
  )
  n3 = VectorParameter() {defaultValue=<rgba 1 0 0 0>}
  n4 = Subtract(A=DotProduct(A=TextureCoordinate(), B=ComponentMask(None=@n3.RGB) {g=true, r=true})) {constB=0.5}
  NamedRerouteDeclaration(None=Saturate(None=Multiply(A=@n4) {constB=256})) {name=CutInHalf, nodeColor=<rgba 0 0 0 1>}
  n5 = ScalarParameter() {parameterName=MinWidth}
  n6 = ScalarParameter() {defaultValue=0.2, group=Outer Skew Box, parameterName=Angle}
  n7 = ScalarParameter() {defaultValue=1, group=Outer Skew Box, parameterName=SkewScale}
  n8 = MaterialFunctionCall!(MinWidth=@n5, Angle Top, Angle Right, Angle Bottom, Angle Left=@n6, Scale=@n7, Flip=StaticBool()) {materialFunction=/Game/UI/Sidebar/Materials/MF_BoxSkew}
  n9 = NamedRerouteUsage()
  n10 = OneMinus(None=@n9)
  n11 = MaterialFunctionCall(MinWidth=@n5, Angle Left=@n6, Scale=@n7, Flip=StaticBool() {value=true}) {materialFunction=&MF_BoxSkew}
  NamedRerouteDeclaration(None=Add(A=LinearInterpolate(B=@n8.Result, Alpha=@n10), B=LinearInterpolate(B=@n9, Alpha=@n11.Result))) {name=Shape, nodeColor=<rgba 0 0 0 1>}
  n12 = TextureCoordinate() {coordinateIndex=3}
  n13 = ComponentMask(None=@n12) {g=true}
  n14 = ComponentMask(None=@n12) {r=true}
  NamedRerouteDeclaration(None=Divide(A=@n13, B=@n14)) {name=ScaleX, nodeColor=<rgba 0 0 0 1>}
  NamedRerouteDeclaration(None=Divide(A=@n14, B=@n13)) {name=ScaleY, nodeColor=<rgba 0 0 0 1>}
  Saturate(None=@n2)
 ]
}
